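# Physics asset for the FaceAcc part of the female GelatoCourage Ramierz3L body cosmetic, covering its hair and ponytail bones.
# Two hair pieces hang off the head with tight swing and twist limits, alongside a four-link ponytail chain with wider swing and minimal twist.
# Capsules on head, neck, upper spine and a shoulder helper bone form the body volume, joined by free angular constraints.
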PhysicsAsset /BRCosmetics/Characters/Player/Female/Medium/Bodies/F_MED_GelatoCourage_Ramierz3L/Meshes/Parts/F_MED_GelatoCourage_Ramierz3L_FaceAcc_Physics
bodies {
 dyn_hairPart_02 [{shapeType=Capsule, length=3.0743, radius=1.9454, shapeName=dyn_hairPart_02_capsule, center=<0 0 -2.6286>}]
 dyn_hairPart_03 [{shapeType=Capsule, length=2.6972, radius=1.2038, shapeName=dyn_hairPart_03_capsule, center=<0 0 -2.6286>}]
 dyn_ponytail_05 [{shapeType=Capsule, length=5.559, radius=2.9861, shapeName=dyn_ponytail_05_capsule, center=<2.7879 0 0>, rotation=<rot 90 0 0>}]
 dyn_ponytail_05_a [{shapeType=Capsule, length=4.3307, radius=4.353, shapeName=dyn_ponytail_05_a_capsule, center=<2.7879 -0.9 1.5>, rotation=<rot 90 0 0>}]
 dyn_ponytail_05_b [{shapeType=Capsule, length=2.1653, radius=2.8617, shapeName=dyn_ponytail_05_b_capsule, center=<2.7879 0 0.7>, rotation=<rot 90 0 0>}]
 dyn_ponytail_05_c [{shapeType=Capsule, length=1.697, radius=1.8933, shapeName=dyn_ponytail_05_c_capsule, center=<3.4536 0 0>, rotation=<rot 90 0 0>}]
 dyn_shouldersetup_MSTR [{shapeType=Capsule, length=5.505, radius=3.505, shapeName=dyn_shouldersetup_MSTR_capsule, center=<0.91 1.5453 2.0628>, rotation=<rot 90 0 0>}]
 head [{shapeType=Capsule, length=7.2787, radius=7.2735, center=<2.5279 -0.8588 0>, rotation=<rot 90 0 0>}]
 neck_02 [{shapeType=Capsule, length=2.625, radius=5.505, shapeName=neck_02_capsule, rotation=<rot 90 0 0>}]
 spine_05 [
  {shapeType=Capsule, length=21.125, radius=8.875, center=<7.7413 0.7602 0>}
  {shapeType=Capsule, length=11.375, radius=10.125, shapeName=spine_05_capsule, center=<2.2506 0.5679 0>, rotation=<rot 90 0 -10.0051>}
 ]
}
constraints [
 {
  bone1Name=neck_02
  bone2Name=spine_05
  swing1LimitDegrees=45
  swing1Motion=Free
  swing2LimitDegrees=45
  swing2Motion=Free
  twistLimitDegrees=45
  twistMotion=Free
 }
 {
  bone1Name=head
  bone2Name=neck_02
  swing1LimitDegrees=45
  swing1Motion=Free
  swing2LimitDegrees=45
  swing2Motion=Free
  twistLimitDegrees=45
  twistMotion=Free
 }
 {bone1Name=dyn_hairPart_02, bone2Name=head, swing1LimitDegrees=2, swing2LimitDegrees=15, twistLimitDegrees=15}
 {bone1Name=dyn_hairPart_03, bone2Name=head, swing1LimitDegrees=2, swing2LimitDegrees=15, twistLimitDegrees=15}
 {bone1Name=dyn_ponytail_05, bone2Name=head, swing1LimitDegrees=45, swing2LimitDegrees=45, twistLimitDegrees=2}
 {bone1Name=dyn_ponytail_05_a, bone2Name=dyn_ponytail_05, swing1LimitDegrees=45, swing2LimitDegrees=45, twistLimitDegrees=2}
 {bone1Name=dyn_ponytail_05_b, bone2Name=dyn_ponytail_05_a, swing1LimitDegrees=45, swing2LimitDegrees=45, twistLimitDegrees=2}
 {bone1Name=dyn_ponytail_05_c, bone2Name=dyn_ponytail_05_b, swing1LimitDegrees=45, swing2LimitDegrees=45, twistLimitDegrees=2}
 {
  bone1Name=dyn_shouldersetup_MSTR
  bone2Name=spine_05
  swing1LimitDegrees=45
  swing1Motion=Free
  swing2LimitDegrees=45
  swing2Motion=Free
  twistLimitDegrees=45
  twistMotion=Free
 }
]
